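# Master material for the Lazuli backpack that layers an animated, emissive bat eye and ear effect onto the 2023 character parent shading.
# Eye and ear distance-field textures, placed by offset, scale and rotation, feed an SDF line function for moving, pulsing lines.
# Tech-art mask channels confine each effect, which a sine envelope or front-end override fades in between the PreFX and PostFX functions.
# Ear highlight follows time of day.
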Material /Game/Accessories/FORT_Backpacks/Backpack_M_MED_Lazuli/Materials/M_Lazuli_Backpack_2023
graph {
 properties {
  bUseMaterialAttributes=true
  bUsedWithMeshDeformer=true
  bUsedWithNanite=true
  bUsedWithSkeletalMesh=true
  pixelDepthOffsetMode=PDOM_Legacy
  shadingModel=MSM_FromMaterialExpression
  usedShadingModels=Default Lit | Subsurface | Clear Coat | Subsurface Profile | Cloth
 }
 nodes [
  Constant() {r=10}
  n1 = VectorParameter() {group=BatEar, parameterName=EarBackGroundColor, sortPriority=40, defaultValue=<rgba 0.5677 0 0.5261 1>}
  n2 = ScalarParameter() {defaultValue=3, group=BatEar, parameterName=Ear_HighLight}
  n3 = Divide(A=@n2)
  n4 = MaterialFunctionCall(Value_Morning=@n2, Value_Day=@n2, Value_Evening=@n3, Value_Night=@n3) {materialFunction=&Scalar_Time_of_Day}
  n5 = Power(Base=OneMinus(None=NamedRerouteUsage()), Exp=ScalarParameter() {defaultValue=5, group=BatEar, parameterName=Ear_Power}) {constExponent=10}
  n6 = Step(
   Y=NamedRerouteDeclaration(None=ScalarParameter() {group=BatEar, parameterName=Ear_ShapeCenterStep, sortPriority=61}) {name=Ear_ShapeCenterStep, nodeColor=<rgba 0.4297 0 1 1>}
   X=NamedRerouteUsage()
  ) {constY=0.1}
  n7 = ScalarParameter() {defaultValue=0.01, group=BatEar, parameterName=Ear_ShapeLineWidthMin, sortPriority=54}
  n8 = MaterialFunctionCall!(
   SDF=NamedRerouteUsage()
   TimeOffset=NamedRerouteUsage()
   LineThickness=@n7
   LineDistance=ScalarParameter() {defaultValue=0.07, group=BatEar, parameterName=Ear_ShapeLineDistance, sortPriority=50}
   VariableThickness=StaticBoolParameter() {defaultValue=true, group=BatEar, parameterName=Ear_ShapeLineVariableThickness, sortPriority=51}
   ThicknessRangeMin=ScalarParameter() {defaultValue=0.3, group=BatEar, parameterName=Ear_ShapeLineRangeMin, sortPriority=52}
   ThicknessRangeMax=ScalarParameter() {group=BatEar, parameterName=Ear_ShapeLineRangeMax, sortPriority=53}
   ThicknessMin=@n7
   ThicknessMax=ScalarParameter() {defaultValue=0.0275, group=BatEar, parameterName=Ear_ShapeLineWidthMax, sortPriority=55}
   Animate=StaticBoolParameter() {defaultValue=true, group=BatEar, parameterName=Ear_ShapeLineAnimate, sortPriority=56}
   Speed=ScalarParameter() {defaultValue=-0.02, group=BatEar, parameterName=Ear_ShapeLineSpeed, sortPriority=57}
   Pulse=StaticBoolParameter() {defaultValue=true, group=BatEar, parameterName=Ear_ShapeLinePulse, sortPriority=58}
   PulseFrequency=NamedRerouteUsage()
   BlurMulti=NamedRerouteUsage()
   PulseSpeed=NamedRerouteUsage()
  ) {materialFunction=&/MF_SDF_Lines_With_Blur}
  n9 = SphereMask(
   A=Saturate(None=OneMinus(None=@n8.Lines))
   B=Constant() {r=1}
   Radius=LinearInterpolate(A=NamedRerouteUsage(), B=Constant() {r=1}, Alpha=ScalarParameter() {defaultValue=1, group=BatEar, parameterName=EarControl})
   Hardness=ScalarParameter() {defaultValue=0.5, group=Bat, parameterName=DF_LineHardness, sliderMax=1}
  ) {attenuationRadius=1}
  n10 = LinearInterpolate(
   A=Multiply(A=@n1.RGB, B=LinearInterpolate(B=@n4.Result, Alpha=@n5) {constA=1, constB=2})
   B=NamedRerouteUsage()
   Alpha=NamedRerouteDeclaration(None=Multiply(A=@n6, B=@n9)) {name=Ear_ShapeLines, nodeColor=<rgba 0.7578 0 1 1>}
  )
  n11 = NamedRerouteUsage()
  LinearInterpolate(A=@n10, B=Multiply(A=@n10, B=@n8.PulseMask), Alpha=Step(X=OneMinus(None=@n11)) {constY=0.1})
  n12 = VectorParameter() {group=Bat, parameterName=EyeBackGroundColor, sortPriority=40, defaultValue=<rgba 0.5677 0 0.5261 1>}
  n13 = NamedRerouteDeclaration(None=@n12.RGB) {name=CachedBackGroundColor, nodeColor=<rgba 1 0.6563 0 1>}
  n14 = LinearInterpolate(Alpha=OneMinus(None=NamedRerouteUsage())) {constA=1, constB=3}
  n15 = VectorParameter() {group=Bat, parameterName=EyeCenterColor, sortPriority=40, defaultValue=<rgba 0.185 0.1833 0.5677 1>}
  n16 = Power(Base=OneMinus(None=NamedRerouteUsage()), Exp=ScalarParameter() {defaultValue=5, parameterName=SDF_CenterPow}) {constExponent=5}
  n17 = NamedRerouteDeclaration(None=SmoothStep(Value=@n16) {constMax=0.2, constMin=0.18}) {name=CenterMask, nodeColor=<rgba 1 0 0.4922 1>}
  n18 = VectorParameter() {group=Bat, parameterName=LineColor, sortPriority=40, defaultValue=<rgba 0.0017 0 0.4062 1>}
  n19 = NamedRerouteDeclaration(None=@n18.RGB) {name=CachedLineColor, nodeColor=<rgba 0 0.9766 1 1>}
  n20 = Multiply(A=@n16) {constB=2}
  n21 = Multiply(A=@n17, B=Sine(None=Frac(None=@n20)))
  n22 = LinearInterpolate(
   A=LinearInterpolate(A=Multiply(A=@n13, B=@n14), B=@n15.RGB, Alpha=@n17)
   B=@n19
   Alpha=NamedRerouteDeclaration(None=Ceil(None=@n21)) {name=CachedNotMoved, nodeColor=<rgba 0 1 0.625 1>}
  )
  n23 = NamedRerouteUsage()
  n24 = Saturate(None=Subtract(A=NamedRerouteUsage(), B=Ceil(None=@n23)))
  n25 = Multiply(
   A=LinearInterpolate(A=@n22, B=@n19, Alpha=@n24)
   B=ScalarParameter() {defaultValue=1, group=Bat, parameterName=ShapeEmisiveStrength, sortPriority=41}
  )
  n26 = NamedRerouteUsage()
  LinearInterpolate(A=@n25, B=Multiply(A=@n25, B=NamedRerouteUsage()), Alpha=Step(X=OneMinus(None=@n26)) {constY=0.1})
  n27 = MaterialFunctionCall() {materialFunction=&MF_CharacterParent_2023_PreFX}
  n28 = GetMaterialAttributes(None=@n27.Result) {attributeGetTypes [#1, #2, #3]}
  n29 = NamedRerouteUsage()
  n30 = ChannelMaskParameter(None=@n29) {group=AdditionalTechArt, maskChannel=Green, parameterName=EyeArea, defaultValue=<rgba 0 1 0 0>}
  n31 = ChannelMaskParameter(None=@n29) {group=AdditionalTechArt, parameterName=EarArea}
  n32 = Add(
   A=Multiply(A=NamedRerouteDeclaration(None=@n30.R) {name=CachedEyeArea, nodeColor=<rgba 0 1 0.1094 1>}, B=NamedRerouteUsage())
   B=Multiply(A=NamedRerouteDeclaration(None=@n31.R) {name=CachedEarArea, nodeColor=<rgba 0.9453 0 1 1>}, B=NamedRerouteUsage())
  )
  n33 = Multiply(A=Sine(None=NamedRerouteUsage())) {constB=3}
  n34 = LinearInterpolate(
   A=NamedRerouteDeclaration(None=Saturate(None=@n33)) {name=CachedGrow, nodeColor=<rgba 0 0.8125 1 1>}
   B=ScalarParameter() {defaultValue=1, group=AdditionalTechArt, parameterName=CCPM_EffectOn}
   Alpha=ScalarParameter() {parameterName=FrontEnd}
  )
  n35 = SetMaterialAttributes!(
   MaterialAttributes=@n28.MaterialAttributes
   Emissive Color=StaticSwitchParameter(True=LinearInterpolate(B=@n32, Alpha=@n34), False=Constant()) {defaultValue=true, group=AdditionalTechArt, parameterName=UseBatEffect}
   Base Color=@n28.MaterialAttributes
   Metallic=@n28.MaterialAttributes
   Roughness=@n28.MaterialAttributes
  ) {attributeSetTypes [
    #4
    #1
    #2
    #3
   ]}
  MaterialFunctionCall(Material=@n35, M=@n27.M, Pre Skinned Local Position=@n27.M, Distance Blend=@n27.M, Fresnel=@n27.M) {materialFunction=&MF_CharacterParent_2023_PostFX}
  n36 = Multiply(A=@n18.RGB, B=@n24)
  Multiply(A=NamedRerouteDeclaration(None=Add(A=@n22, B=@n36)) {name=CachedBackGround, nodeColor=<rgba 0.6641 0 1 1>})
  NamedRerouteDeclaration(None=@n27.Tech Art Mask) {name=CachedTechArtMask, nodeColor=<rgba 1 0 0.3281 1>}
  MaterialExpressionNamedRerouteUsage_13 = NamedRerouteUsage()
  n37 = Constant2Vector() {g=0.5, r=0.5}
  MaterialExpressionAdd_9 = Add(
   A=AppendVector(
    A=ScalarParameter() {group=BatEar, parameterName=Shape_UVOffsetX02, sortPriority=62}
    B=ScalarParameter() {group=BatEar, parameterName=Shape_UVOffsetY02, sortPriority=63}
   )
   B=@n37
  )
  MaterialExpressionScalarParameter_3 = ScalarParameter() {defaultValue=1, group=BatEar, parameterName=Shape_UVScale02, sortPriority=65}
  n38 = Custom!(OriginalUV=@MaterialExpressionNamedRerouteUsage_13, UVPoint=@MaterialExpressionAdd_9, UVScale=@MaterialExpressionScalarParameter_3) {
   code="float2 DistUV =OriginalUV-UVPoint;\r\nDistUV = DistUV/(UVScale/2) + float2(0.5,0.5);\r\n\r\nreturn DistUV;"
   description=CopyToPointUV
   outputType=CMOT_Float2
   inputs [
    {
     inputName=OriginalUV
     input {expression=&self:MaterialExpressionNamedRerouteUsage_13}
    }
    {
     inputName=UVPoint
     input {expression=&self:MaterialExpressionAdd_9}
    }
    {
     inputName=UVScale
     input {expression=&self:MaterialExpressionScalarParameter_3}
    }
   ]
  }
  NamedRerouteDeclaration(None=MaterialFunctionCall(
    UVs=Clamp(None=@n38)
    Rotation Center=@n37
    "Rotation Angle (0-1)"=ScalarParameter() {group=BatEar, parameterName=Shape_UVRotation02, sortPriority=64}
   ) {materialFunction=&CustomRotator}.Rotated Values) {name=ShapeUV02, nodeColor=<rgba 0 0.7656 1 1>}
  MaterialExpressionNamedRerouteDeclaration_2 = NamedRerouteDeclaration(None=StaticSwitchParameter(True=TextureCoordinate() {coordinateIndex=1}, False=TextureCoordinate()) {group=Bat, parameterName=UseUV1}) {name=UVs, nodeColor=<rgba 0 0.3672 1 1>}
  n39 = TextureObjectParameter(UVs=@MaterialExpressionNamedRerouteDeclaration_2) {
   group=BatEar
   parameterName=Shape02
   samplerType=SAMPLERTYPE_LinearColor
   sortPriority=-10
   texture=/Game/Accessories/FORT_Backpacks/Backpack_M_MED_Lazuli/Textures/T_BatEar_DF
  }
  NamedRerouteDeclaration(None=@n39) {name=CachedShapeTexture02, nodeColor=<rgba 1 0.8203 0 1>}
  n40 = NamedRerouteUsage()
  n41 = NamedRerouteUsage()
  NamedRerouteDeclaration(None=OneMinus(None=TextureSample(UVs=@n40, Tex=@n41) {samplerSource=SSM_Wrap_WorldGroupSettings, samplerType=SAMPLERTYPE_LinearColor}.R)) {name=Ear_Shape_SDF, nodeColor=<rgba 0.5703 0 1 1>}
  NamedRerouteDeclaration(None=@n10) {name=CachedEarColor, nodeColor=<rgba 0.8281 1 0 1>}
  n42 = Divide(
   A=Subtract(A=PixelDepth().R, B=ScalarParameter() {defaultValue=50, group=Bat, parameterName=BlurStartDistance})
   B=ScalarParameter() {defaultValue=500, group=Bat, parameterName=BlurFadeDistance}
  )
  NamedRerouteDeclaration(None=LinearInterpolate(B=ScalarParameter() {defaultValue=0.5, group=Bat, parameterName=BlurMinThick}, Alpha=Saturate(None=@n42)) {constA=1}) {name=DepthDrivenRadius, nodeColor=<rgba 0.7578 1 0 1>}
  n43 = Subtract(A=Time(), B=ScalarParameter() {group=MF_AdvancedEM, parameterName=CCPM_End_Time})
  n44 = ScalarParameter() {defaultValue=10, group=MF_AdvancedEM, parameterName=CCPM_Time_Length}
  NamedRerouteDeclaration(None=Saturate(None=Divide(A=@n43, B=@n44))) {name=CachedCCPMTime, nodeColor=<rgba 0.6172 1 0 1>}
  n45 = ScalarParameter() {group=Bat, parameterName=TimeOffset}
  n46 = NamedRerouteDeclaration(None=ScalarParameter() {defaultValue=0.01, group=Bat, parameterName=ShapeLineWidthMin, sortPriority=54}) {name=ShapeLineWidthMin, nodeColor=<rgba 0 0.6686 1 1>}
  n47 = MaterialFunctionCall(
   SDF=NamedRerouteUsage()
   TimeOffset=@n45
   LineThickness=@n46
   LineDistance=NamedRerouteDeclaration(None=ScalarParameter() {defaultValue=0.2, group=Bat, parameterName=ShapeLineDistance, sortPriority=50}) {name=ShapeLineDistance, nodeColor=<rgba 0 0.6686 1 1>}
   VariableThickness=NamedRerouteDeclaration(None=StaticBoolParameter() {group=Bat, parameterName=ShapeLineVariableThickness, sortPriority=51}) {name=ShapeLineVariableThickness, nodeColor=<rgba 0 0.4609 1 1>}
   ThicknessRangeMin=NamedRerouteDeclaration(None=ScalarParameter() {defaultValue=0.3, group=Bat, parameterName=ShapeLineRangeMin, sortPriority=52}) {name=ShapeLineRangeMin, nodeColor=<rgba 0 0.6686 1 1>}
   ThicknessRangeMax=NamedRerouteDeclaration(None=ScalarParameter() {group=Bat, parameterName=ShapeLineRangeMax, sortPriority=53}) {name=ShapeLineRangeMax, nodeColor=<rgba 0 0.6686 1 1>}
   ThicknessMin=@n46
   ThicknessMax=NamedRerouteDeclaration(None=ScalarParameter() {defaultValue=0.0275, group=Bat, parameterName=ShapeLineWidthMax, sortPriority=55}) {name=ShapeLineWidthMax, nodeColor=<rgba 0 0.6686 1 1>}
   Animate=NamedRerouteDeclaration(None=StaticBoolParameter() {defaultValue=true, group=Bat, parameterName=ShapeLineAnimate, sortPriority=56}) {name=ShapeLineAnimate, nodeColor=<rgba 0 0.7891 1 1>}
   Speed=NamedRerouteDeclaration(None=ScalarParameter() {defaultValue=-0.02, group=Bat, parameterName=ShapeLineSpeed, sortPriority=57}) {name=ShapeLineSpeed, nodeColor=<rgba 0.8516 1 0 1>}
   Pulse=NamedRerouteDeclaration(None=StaticBoolParameter() {defaultValue=true, group=Bat, parameterName=ShapeLinePulse, sortPriority=58}) {name=ShapeLinePulse, nodeColor=<rgba 0 0.2266 1 1>}
   PulseFrequency=NamedRerouteDeclaration(None=ScalarParameter() {defaultValue=3, group=Bat, parameterName=ShapeLinePulseFrequency, sortPriority=59}) {name=ShapeLinePulseFrequency, nodeColor=<rgba 0 0.6686 1 1>}
   BlurMulti=NamedRerouteUsage()
   PulseSpeed=NamedRerouteDeclaration(None=ScalarParameter() {defaultValue=-0.65, group=Bat, parameterName=ShapeLinePulseSpeed, sortPriority=60}) {name=ShapeLinePulseSpeed, nodeColor=<rgba 0 0.6686 1 1>}
  ) {materialFunction=&MF_SDF_Lines_With_Blur}
  NamedRerouteDeclaration(None=@n47.PulseMask) {name=CachedPulse, nodeColor=<rgba 1 0.1172 0 1>}
  NamedRerouteDeclaration(None=@n25) {name=ShapeColor, nodeColor=<rgba 0.8281 1 0 1>}
  NamedRerouteDeclaration(None=@n45) {name=CachedTimeOffset, nodeColor=<rgba 0 0.8594 1 1>}
  NamedRerouteDeclaration(None=TextureSample(UVs=NamedRerouteUsage(), Tex=NamedRerouteUsage()) {samplerSource=SSM_Wrap_WorldGroupSettings, samplerType=SAMPLERTYPE_LinearColor}.R) {name=Eye_Shape_SDF, nodeColor=<rgba 0.5703 0 1 1>}
  n48 = NamedRerouteDeclaration(None=ScalarParameter() {group=Bat, parameterName=ShapeCenterStep, sortPriority=61}) {name=ShapeCenterStep, nodeColor=<rgba 0.4297 0 1 1>}
  n49 = SphereMask(
   A=Saturate(None=OneMinus(None=@n47.Lines))
   B=Constant() {r=1}
   Radius=NamedRerouteUsage()
   Hardness=ScalarParameter() {defaultValue=0.5, group=Bat, parameterName=DF_LineHardness, sliderMax=1}
  ) {attenuationRadius=1}
  NamedRerouteDeclaration(None=Multiply(A=Step(Y=@n48, X=@n49) {constY=0.1}, B=@n49)) {name=ShapeLines, nodeColor=<rgba 0.7578 0 1 1>}
  NamedRerouteDeclaration(None=ScalarParameter() {defaultValue=20, group=Bat, parameterName=BlurMaxForTAA}) {name=DrivenBlur, nodeColor=<rgba 0.0312 0 1 1>}
  n50 = Constant2Vector() {g=0.5, r=0.5}
  MaterialExpressionAdd_3 = Add(
   A=AppendVector(
    A=ScalarParameter() {group=Bat, parameterName=Shape_UVOffsetX01, sortPriority=62}
    B=ScalarParameter() {group=Bat, parameterName=Shape_UVOffsetY01, sortPriority=63}
   )
   B=@n50
  )
  MaterialExpressionScalarParameter_47 = ScalarParameter() {defaultValue=1, group=Bat, parameterName=Shape_UVScale01, sortPriority=65}
  n51 = Custom(OriginalUV=@MaterialExpressionNamedRerouteDeclaration_2, UVPoint=@MaterialExpressionAdd_3, UVScale=@MaterialExpressionScalarParameter_47) {
   code="float2 DistUV =OriginalUV-UVPoint;\r\nDistUV = DistUV/(UVScale/2) + float2(0.5,0.5);\r\n\r\nreturn DistUV;"
   description=CopyToPointUV
   outputType=CMOT_Float2
   inputs [
    {
     inputName=OriginalUV
     input {expression=&self:MaterialExpressionNamedRerouteDeclaration_2}
    }
    {
     inputName=UVPoint
     input {expression=&self:MaterialExpressionAdd_3}
    }
    {
     inputName=UVScale
     input {expression=&self:MaterialExpressionScalarParameter_47}
    }
   ]
  }
  NamedRerouteDeclaration(None=MaterialFunctionCall(
    UVs=Clamp(None=@n51)
    Rotation Center=@n50
    "Rotation Angle (0-1)"=ScalarParameter() {group=Bat, parameterName=Shape_UVRotation01, sortPriority=64}
   ) {materialFunction=&CustomRotator}.Rotated Values) {name=ShapeUV01, nodeColor=<rgba 0 0.7656 1 1>}
  n52 = TextureObjectParameter(UVs=@MaterialExpressionNamedRerouteDeclaration_2) {
   group=Bat
   parameterName=Shape01
   samplerType=SAMPLERTYPE_LinearColor
   sortPriority=-10
   texture=/Game/Accessories/FORT_Backpacks/Backpack_M_MED_Lazuli/Textures/T_BatEye_DF
  }
  NamedRerouteDeclaration(None=@n52) {name=CachedShapeTexture01, nodeColor=<rgba 1 0.8203 0 1>}
  NamedRerouteUsage()
  ScalarParameter() {defaultValue=20, group=Bat, parameterName=BlurMinForTAA}
  StaticBoolParameter() {defaultValue=true, group=Bat, parameterName=PulseSpeed, sortPriority=51}
 ]
}
